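# Physics asset for the Tech Ops pickaxe skeletal mesh, a two-handed melee weapon cosmetic.
# It holds a single capsule body on the Melee_TwoHanded_Handle bone, offset and rotated from the bone origin.
# There are no constraints, so the asset defines one rigid collision shape with no jointed parts.
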PhysicsAsset /BRCosmetics/Weapons/FORT_Melee/Pickaxe_Tech_Ops/Meshes/SK_Pickaxe_Tech_Ops_PhysicsAsset
bodies {Melee_TwoHanded_Handle [{shapeType=Capsule, length=83.4579, radius=34.9289, center=<0.9591 12.0662 36.9488>, rotation=<rot 3.2177 -90 -0.3672>}]}
constraints []
